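# Niagara idle effect for the Circuit Breaker pickaxe, with a single ElectricArcs emitter that spawns mesh particles at a low rate.
# Each short-lived arc spawns at a fixed offset with random material values for variation.
# Over its lifetime, a curve sweeps another dynamic material parameter downward, likely animating the arc, while a color curve dims its intense orange-red HDR tint.
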NiagaraSystem /BRCosmetics/Weapons/FORT_Melee/Pickaxe_CircuitBreaker/NS_Pickaxe_Idle_Circuitbreaker
dependencies {
 usedDataInterfaces [&NiagaraDataInterfaceCurve, &NiagaraDataInterfaceVectorCurve]
 usedDynamicInputs []
 usedModules [&SystemState]
}
emitters {ElectricArcs {
  emitterUpdateScript [
   &EmitterState(Life Cycle Mode=::NewEnumerator1, Loop Behavior=::NewEnumerator1, Loop Duration Mode=::NewEnumerator1, UseLoopDelay=false)
   &SpawnRate(SpawnRate=4, Use Spawn Probability=false)
  ]
  particleSpawnScript [
   &InitializeParticle(
    Color=<rgba 3 3 3 0.6>
    Lifetime=&RandomRangeFloat(Minimum=0.5, Maximum=1, Evaluation Type=::NewEnumerator1)
    Mesh Scale Mode=::NewEnumerator1
    Mesh Uniform Scale Max=1.1
    Position Offset=<1 -14 53>
    UsePositionOffset=true
    Write Material Random=true
   )
   &DynamicMaterialParameters(
    Index 0 Param 1=&RandomRangeFloat(Minimum=0.7, Maximum=1.5, Evaluation Type=::NewEnumerator1)
    Index 0 Param 3=&RandomRangeFloat(Minimum=0.1, Maximum=0.4, Evaluation Type=::NewEnumerator1)
    Module.Param0WriteEnabled=true
    Module.Param1WriteEnabled=false
    Module.Param2WriteEnabled=true
    Module.Param3WriteEnabled=true
   )
  ]
  particleUpdateScript [
   &ParticleState()
   &DynamicMaterialParameters(
    Index 0 Param 2=&FloatFromCurve(FloatCurve=DataInterface() {propertyValues {Curve {keys [
         {arriveTangent=-4.3617, interpMode=RCIM_Cubic, leaveTangent=-4.3617, tangentMode=RCTM_User, time=0.1129, value=1}
         {arriveTangent=-0.9065, leaveTangent=-0.9065, time=1, value=-1}
        ]}}})
    Module.Param0WriteEnabled=false
    Module.Param1WriteEnabled=true
    Module.Param2WriteEnabled=false
    Module.Param3WriteEnabled=false
   ) {module=DynamicMaterialParameters001}
   &ScaleColor(
    Scale RGB=&VectorFromCurve(VectorCurve=DataInterface() {propertyValues {XCurve {keys [<at 0.2 50>,<at 1 5>]}, YCurve {keys [<at 0.2 10>,<at 1 1>]}, ZCurve {keys [<at 0.2 0>,<at 1 0>]}}})
    ScaleA=true
    ScaleRGB=true
   )
  ]
  rendererClasses [&NiagaraMeshRendererProperties]
  renderers [{rendererClass=&NiagaraMeshRendererProperties, rendererIndex=0}]
 }}
summary {
 systemName=NS_Pickaxe_Idle_Circuitbreaker
 emitters [{emitterName=ElectricArcs, rendererClasses [&NiagaraMeshRendererProperties]}]
 userVariables []
}
